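# Looping spot ambience for the RustBucket brazier fire, played quietly with small-radius attenuation, occlusion and an exterior ambient bus send.
# A mixer layers the main fire loop, one of three random thickener loops and a reused Coliseum brazier stereo loop.
# Random pitch modulation and two looping envelopes of different lengths add slow pitch and volume drift.
# The root randomly starts this chain immediately or after a half-second delay, which can stagger instances.
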SoundCue /Game/Sounds/Helios/Environment_Actors/RustBucketBrazierFire/Spot_RustBucket_Brazier_Fire_Loop_Cue
root=Random_1
nodes [
 {id=Delay_0, children [Enveloper_3], properties {delayMax=0.5, delayMin=0.5}}
 {
  id=Enveloper_0
  children [Modulator_0]
  properties {
   bLoop=true
   bLoopIndefinitely=true
   loopEnd=30
   pitchCurve {editorCurveData {keys [{arriveTangent=-0.0013, interpMode=RCIM_Cubic, time=0}, <at 15 1.05 RCIM_Cubic>, {interpMode=RCIM_Cubic, leaveTangent=0.0013, time=30}]}}
   volumeCurve {editorCurveData {keys [<at 0 1>]}}
  }
 }
 {
  id=Enveloper_3
  children [Enveloper_0]
  properties {
   bLoop=true
   bLoopIndefinitely=true
   loopEnd=20
   pitchCurve {editorCurveData {keys [{time=0}]}}
   volumeCurve {editorCurveData {keys [
      {arriveTangent=-0.01, interpMode=RCIM_Cubic, leaveTangent=-0.01, time=0, value=1}
      <at 10 0.9 RCIM_Cubic>
      {arriveTangent=0.01, interpMode=RCIM_Cubic, leaveTangent=0.01, time=20, value=1}
     ]}}
  }
 }
 {id=Mixer_2, children [WavePlayer_3,Random_0,WavePlayer_9], properties {inputVolume [1.2,1,0.75]}}
 {id=Modulator_0, children [Mixer_2], properties {pitchMax=1.1, pitchMin=1}}
 {id=Random_0, children [WavePlayer_6,WavePlayer_7,WavePlayer_8], properties {weights [1,1,1]}}
 {id=Random_1, children [Delay_0,Enveloper_3], properties {weights [1,1]}}
 {
  id=WavePlayer_3
  properties {bLooping=true, soundWaveAssetPtr=&/Spot_RustBucket_Brazier_Fire_Loop}
 }
 {id=WavePlayer_6, properties {bLooping=true, soundWaveAssetPtr=/Game/Sounds/Helios/Environment_Actors/Brazier/Spot_Fire_Thickener_Loop_01}}
 {id=WavePlayer_7, properties {bLooping=true, soundWaveAssetPtr=/Game/Sounds/Helios/Environment_Actors/Brazier/Spot_Fire_Thickener_Loop_02}}
 {id=WavePlayer_8, properties {bLooping=true, soundWaveAssetPtr=/Game/Sounds/Helios/Environment_Actors/Brazier/Spot_Fire_Thickener_Loop_03}}
 {
  id=WavePlayer_9
  properties {bLooping=true, soundWaveAssetPtr=/Game/Sounds/Artemis/Environment_Actors/ColiseumBrazierFire/Spot_Coliseum_Brazier_Fire_Stereo}
 }
]
properties {
 attenuationSettings=&/Spot_RustBucket_Brazier_Small_Att
 bOverrideConcurrency=true
 maxDistance=800
 soundClassObject=&Fort_Gadgets
 subtitlePriority=5000
 volumeMultiplier=0.1
 attenuationOverrides {
  bEnableFocusInterpolation=true
  bEnableListenerFocus=true
  bEnableOcclusion=true
  dBAttenuationAtMax=-40
  falloffDistance=512
  focusAzimuth=15
  focusVolumeAttenuation=1.5
  nonFocusAzimuth=45
  nonFocusVolumeAttenuation=0.5
  occlusionInterpolationTime=1
  occlusionLowPassFilterFrequency=10000
  occlusionTraceChannel=ECC_GameTraceChannel11
  occlusionVolumeAttenuation=0.5
  stereoSpread=512
  attenuationShapeExtents=<256 64.5942 90>
 }
 busSends [{audioBus=null, minSendLevel=0.6, sendLevel=1, soundSourceBus=&AmbientAudio_Exterior_SourceBus}]
 concurrencyOverrides {bLimitToOwner=true, maxCount=3}
 soundSubmixSends [{minSendLevel=0.6, soundSubmix=&Fortnite_Reverb_Default}]
}
